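# Control Rig for the F_MED_Modern_Witch player body that correctively rotates the right collar bone, dyn_collar_r, as the right arm moves.
# A spherical pose reader measures upperarm_r against spine_05 and outputs the blend weight.
# That weight drives a reusable "Set Transforms for Bone" function, which adds translation, Euler rotation and scale offsets to the bone's local transform.
# Debug drawing shows the reader and the collar's axes before and after the adjustment.
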ControlRigBlueprint /Game/Characters/Player/Female/Medium/Bodies/F_MED_Modern_Witch/Meshes/F_MED_Modern_Witch_CtrlRig
graphs [
 ContainedGraph {nodes [
   n1 = RigVMFunctionEntryNode()
   RigUnit_GetTransform = &RigUnit_GetTransform(Space=LocalSpace, bInitial=false, Item=@n1.Joint)
   n2 = &RigVMFunction_MathQuaternionToEuler(RotationOrder=ZYX, Value=@RigUnit_GetTransform.Transform.Rotation)
   > RigUnit_SetTransform = &RigUnit_SetTransform(
    Space=LocalSpace
    Value=(Rotation=(X=0.0132,Y=0.4099,Z=-0.0222,W=0.9118),Translation=(X=23.6045,Y=-6.37,Z=140.1115),Scale3D=(X=1,Y=1,Z=1))
    bInitial=false
    Item=@n1.Joint
    Weight=@n1.Weight
    Value.Translation.Z=&RigVMFunction_MathFloatAdd(B=@n1.Translate Z, A=@RigUnit_GetTransform.Transform.Translation.Z) {template="Add::Execute(in A,in B,out Result)"}.Result
    Value.Scale3D.Z=&RigVMFunction_MathFloatAdd(B=@n1.Scale Z, A=@RigUnit_GetTransform.Transform.Scale3D.Z) {template="Add::Execute(in A,in B,out Result)"}.Result
    Value.Translation.Y=&RigVMFunction_MathFloatAdd(B=@n1.Translate Y, A=@RigUnit_GetTransform.Transform.Translation.Y) {template="Add::Execute(in A,in B,out Result)"}.Result
    Value.Scale3D.Y=&RigVMFunction_MathFloatAdd(B=@n1.Scale Y, A=@RigUnit_GetTransform.Transform.Scale3D.Y) {template="Add::Execute(in A,in B,out Result)"}.Result
    Value.Translation.X=&RigVMFunction_MathFloatAdd(B=@n1.Translate X, A=@RigUnit_GetTransform.Transform.Translation.X) {template="Add::Execute(in A,in B,out Result)"}.Result
    Value.Scale3D.X=&RigVMFunction_MathFloatAdd(B=@n1.Scale X, A=@RigUnit_GetTransform.Transform.Scale3D.X) {template="Add::Execute(in A,in B,out Result)"}.Result
    Value.Rotation=&RigVMFunction_MathQuaternionFromEuler(
     Euler=(X=0,Y=0,Z=0)
     Euler.X=&RigVMFunction_MathFloatAdd(B=@n1.Rotate X, A=@n2.Result.X) {template="Add::Execute(in A,in B,out Result)"}.Result
     Euler.Y=&RigVMFunction_MathFloatAdd(B=@n1.Rotate Y, A=@n2.Result.Y) {template="Add::Execute(in A,in B,out Result)"}.Result
     Euler.Z=&RigVMFunction_MathFloatAdd(B=@n1.Rotate Z, A=@n2.Result.Z) {template="Add::Execute(in A,in B,out Result)"}.Result
    ).Result
   )
   > RigVMFunctionReturnNode()
  ]}
 RigVMFunctionLibrary {nodes [Set Transforms for Bone = RigVMCollapseNode(
    ExecuteContext=()
    Joint=()
    Rotate X=0
    Rotate Y=0
    Rotate Z=0
    Scale X=0
    Scale Y=0
    Scale Z=0
    Translate X=0
    Translate Y=0
    Translate Z=0
    Weight=1
   )]}
 RigVMModel {nodes [
   RigUnit_BeginExecution = &RigUnit_BeginExecution()
   /Script/ControlRig.RigUnit_DebugTransformMutableItemSpace(
    Color=(R=1,G=0,B=0,A=1)
    DebugDrawSettings=(DepthPriority=SDPG_Foreground,Lifetime=-1)
    Mode=Axes
    Scale=5
    Space=(Type=Bone)
    Thickness=0.1
    WorldOffset=(Rotation=(X=0,Y=0,Z=0,W=1),Translation=(X=0,Y=0,Z=0),Scale3D=(X=1,Y=1,Z=1))
    bEnabled=True
    ExecutePin=@RigUnit_BeginExecution.ExecutePin
    Transform=&RigUnit_GetTransform(Item=(Type=Bone,Name="dyn_collar_r")).Transform
   ) {is=struct RigVMUnitNode}
   > n1 = &RigUnit_SphericalPoseReader(
    ActiveRegionScaleFactors=(PositiveWidth=1,NegativeWidth=0.9,PositiveHeight=0.7,NegativeHeight=0.7)
    ActiveRegionSize=0.15
    Debug=(bDrawDebug=True,bDraw2D=false,bDrawLocalAxes=true,DebugScale=25,DebugSegments=20,DebugThickness=0.25)
    DriverAxis=(X=-1,Y=0,Z=0)
    DriverItem=(Type=Bone,Name="upperarm_r")
    FalloffRegionScaleFactors=(PositiveWidth=0.7,NegativeWidth=0.7,PositiveHeight=0.7,NegativeHeight=0.7)
    FalloffSize=0.45
    FlipWidthScaling=true
    OptionalParentItem=(Type=Bone,Name="spine_05")
    RotationOffset=(X=-90,Y=-50,Z=30)
   )
   > Set Transforms for Bone = RigVMFunctionReferenceNode(
    Joint=(Type=Bone,Name="dyn_collar_r")
    Rotate X=-50
    Rotate Y=30
    Rotate Z=0
    Scale X=0
    Scale Y=0
    Scale Z=0
    Translate X=0
    Translate Y=0
    Translate Z=0
    Weight=@n1.OutputParam
   )
   > &RigUnit_DebugTransformMutableItemSpace(
    Color=(R=1,G=0,B=0,A=1)
    DebugDrawSettings=(DepthPriority=SDPG_Foreground,Lifetime=-1)
    Mode=Axes
    Scale=3
    Space=(Type=Bone)
    Thickness=0.3
    WorldOffset=(Rotation=(X=0,Y=0,Z=0,W=1),Translation=(X=0,Y=0,Z=0),Scale3D=(X=1,Y=1,Z=1))
    bEnabled=True
    Transform=&RigUnit_GetTransform(Item=(Type=Bone,Name="dyn_collar_r")).Transform
   )
  ]}
]
hierarchy [
 root {loc=<0 0 0>, rot=<0 0 0 1>}
 pelvis {parent=root, loc=<0 -0.5301 98.0571>, rot=<0.0127 0.707 -0.0127 -0.707>}
 spine_01 {parent=pelvis, loc=<6.1208 -0.9328 0>, rot=<0 0 0.034 -0.9994>}
 spine_02 {parent=spine_01, loc=<6.5996 -0.2095 0>, rot=<0 0 0.0238 -0.9997>}
 spine_03 {parent=spine_02, loc=<6.4914 -0.1237 0>, rot=<0 0 0.0149 -0.9999>}
 spine_04 {parent=spine_03, loc=<6.9598 0.0794 0>, rot=<0 0 -0.0132 -0.9999>}
 spine_05 {parent=spine_04, loc=<6.9987 0 0>, rot=<0 0 -0.0015 -1>}
 clavicle_l {parent=spine_05, loc=<5.4831 -0.1053 -3.2946>, rot=<0.0465 -0.7289 -0.009 -0.6829>}
 upperarm_l {parent=clavicle_l, loc=<9.8658 -1.2136 -0.4105>, rot=<0.0164 -0.3245 -0.0269 -0.9454>}
 lowerarm_l {parent=upperarm_l, loc=<26.465 -0.8345 0.259>, rot=<0.0036 0.0161 -0.2298 -0.9731>}
 hand_l {parent=lowerarm_l, loc=<22.7355 -0.3203 -0.6106>, rot=<-0.6784 -0.0223 0.0735 -0.7307>}
 index_metacarpal_l {parent=hand_l, loc=<1.6535 0.0183 -2.96>, rot=<0.0158 -0.0124 -0.0173 -0.9996>}
 index_01_l {parent=index_metacarpal_l, loc=<8.3008 0.0357 -0.0524>, rot=<-0.0044 0.0485 0.0924 -0.9945>}
 index_02_l {parent=index_01_l, loc=<3.3587 -0.3004 0.0452>, rot=<-0.002 0.0027 0.3225 -0.9466>}
 index_03_l {parent=index_02_l, loc=<2.5578 -0.123 -0.0252>, rot=<-0.002 0.0067 0.2558 -0.9667>}
 middle_metacarpal_l {parent=hand_l, loc=<2.1856 -0.1725 -0.5984>, rot=<0.006 0.0168 -0.0329 -0.9993>}
 middle_01_l {parent=middle_metacarpal_l, loc=<7.3275 -0.2447 0.0129>, rot=<-0.0208 0.0042 0.0858 -0.9961>}
 middle_02_l {parent=middle_01_l, loc=<4.0123 0.1401 0.1175>, rot=<-0.0218 0.0258 0.3512 -0.9357>}
 middle_03_l {parent=middle_02_l, loc=<2.6077 0.0534 0.1373>, rot=<0.0082 0.0188 0.1708 -0.9851>}
 pinky_metacarpal_l {parent=hand_l, loc=<1.7616 -0.9753 3.3814>, rot=<-0.0397 0.0814 -0.069 -0.9935>}
 pinky_01_l {parent=pinky_metacarpal_l, loc=<7.1709 -0.4831 -0.1094>, rot=<-0.0654 -0.0035 0.2242 -0.9723>}
 pinky_02_l {parent=pinky_01_l, loc=<2.8274 0.0893 -0.1043>, rot=<0.0037 0.0011 0.2854 -0.9584>}
 pinky_03_l {parent=pinky_02_l, loc=<2.2046 0.0307 -0.0078>, rot=<-0.0002 0.0098 0.1445 -0.9895>}
 ring_metacarpal_l {parent=hand_l, loc=<2.2074 -0.3625 1.3638>, rot=<-0.0284 0.0722 -0.0458 -0.9959>}
 ring_01_l {parent=ring_metacarpal_l, loc=<7.0295 -0.6197 -0.1446>, rot=<-0.0489 -0.0252 0.1892 -0.9804>}
 ring_02_l {parent=ring_01_l, loc=<3.6174 0.4647 -0.1027>, rot=<0 0 0.2806 -0.9598>}
 ring_03_l {parent=ring_02_l, loc=<2.4152 0.114 0.0199>, rot=<0 0 0.1189 -0.9929>}
 thumb_01_l {parent=hand_l, loc=<1.3118 -0.3191 -2.459>, rot=<0.6546 -0.4042 0.0664 -0.6354>}
 thumb_02_l {parent=thumb_01_l, loc=<5.9336 -0.0448 0.0401>, rot=<-0.0037 0.0213 0.298 -0.9543>}
 thumb_03_l {parent=thumb_02_l, loc=<2.2223 0.3682 -0.0907>, rot=<0.0118 -0.013 0.0066 -0.9998>}
 weapon_l {parent=hand_l, loc=<-0.1974 -3.4111 0.3811>, rot=<0 0 0 -1>}
 lowerarm_twist_01_l {parent=lowerarm_l, loc=<18.316 0 0>, rot=<0 0 0 -1>}
 dyn_sleeve_strap_l_MSTR {parent=lowerarm_twist_01_l, loc=<0.2358 -2.5224 3.5805>, rot=<-0.1123 0.1653 -0.6288 -0.7514>}
 dyn_sleeve_strap_l_01 {parent=dyn_sleeve_strap_l_MSTR, loc=<0.3427 0 0>, rot=<0 0 0 1>}
 dyn_sleeve_strap_l_02 {parent=dyn_sleeve_strap_l_01, loc=<3.7156 0 0>, rot=<0 0 0 1>}
 lowerarm_twist_02_l {parent=lowerarm_l, loc=<10.6707 0 0>, rot=<0 0 0 -1>}
 upperarm_twist_01_l {parent=upperarm_l, loc=<7.2135 -0.1134 0>, rot=<0 0 0 1>}
 upperarm_twist_02_l {parent=upperarm_l, loc=<19.7998 -0.1134 0>, rot=<0 0 0 1>}
 clavicle_r {parent=spine_05, loc=<5.4831 -0.1053 3.2946>, rot=<0.7289 0.0465 -0.6829 0.009>}
 upperarm_r {parent=clavicle_r, loc=<-9.8658 1.2136 0.4104>, rot=<0.0164 -0.3245 -0.0269 -0.9454>}
 lowerarm_r {parent=upperarm_r, loc=<-26.4651 0.8345 -0.259>, rot=<0.0036 0.0161 -0.2298 -0.9731>}
 hand_r {parent=lowerarm_r, loc=<-22.7356 0.3203 0.6106>, rot=<-0.6784 -0.0223 0.0735 -0.7307>}
 index_metacarpal_r {parent=hand_r, loc=<-1.6534 -0.0181 2.9601>, rot=<0.0158 -0.0124 -0.0173 -0.9996>}
 index_01_r {parent=index_metacarpal_r, loc=<-8.3008 -0.0357 0.0524>, rot=<-0.0044 0.0485 0.0924 -0.9945>}
 index_02_r {parent=index_01_r, loc=<-3.3589 0.3003 -0.0453>, rot=<-0.002 0.0027 0.3225 -0.9466>}
 index_03_r {parent=index_02_r, loc=<-2.5578 0.123 0.0252>, rot=<-0.002 0.0067 0.2558 -0.9667>}
 middle_metacarpal_r {parent=hand_r, loc=<-2.1856 0.1726 0.5984>, rot=<0.006 0.0168 -0.0329 -0.9993>}
 middle_01_r {parent=middle_metacarpal_r, loc=<-7.3273 0.2447 -0.0129>, rot=<-0.0208 0.0042 0.0858 -0.9961>}
 middle_02_r {parent=middle_01_r, loc=<-4.0124 -0.14 -0.1175>, rot=<-0.0218 0.0258 0.3512 -0.9357>}
 middle_03_r {parent=middle_02_r, loc=<-2.6077 -0.0533 -0.1373>, rot=<0.0082 0.0188 0.1708 -0.9851>}
 pinky_metacarpal_r {parent=hand_r, loc=<-1.7617 0.9753 -3.3814>, rot=<-0.0397 0.0814 -0.069 -0.9935>}
 pinky_01_r {parent=pinky_metacarpal_r, loc=<-7.1707 0.4832 0.1094>, rot=<-0.0654 -0.0035 0.2242 -0.9723>}
 pinky_02_r {parent=pinky_01_r, loc=<-2.8274 -0.0892 0.1043>, rot=<0.0037 0.0011 0.2854 -0.9584>}
 pinky_03_r {parent=pinky_02_r, loc=<-2.2045 -0.031 0.0078>, rot=<-0.0002 0.0098 0.1445 -0.9895>}
 ring_metacarpal_r {parent=hand_r, loc=<-2.2073 0.3627 -1.3638>, rot=<-0.0284 0.0722 -0.0458 -0.9959>}
 ring_01_r {parent=ring_metacarpal_r, loc=<-7.0297 0.6197 0.1446>, rot=<-0.0489 -0.0252 0.1892 -0.9804>}
 ring_02_r {parent=ring_01_r, loc=<-3.6171 -0.4648 0.1027>, rot=<0 0 0.2806 -0.9598>}
 ring_03_r {parent=ring_02_r, loc=<-2.4153 -0.1141 -0.0199>, rot=<0 0 0.1189 -0.9929>}
 thumb_01_r {parent=hand_r, loc=<-1.3119 0.3192 2.459>, rot=<0.6546 -0.4042 0.0664 -0.6354>}
 thumb_02_r {parent=thumb_01_r, loc=<-5.9335 0.0447 -0.0401>, rot=<-0.0037 0.0213 0.298 -0.9543>}
 thumb_03_r {parent=thumb_02_r, loc=<-2.2223 -0.3682 0.0907>, rot=<0.0118 -0.013 0.0066 -0.9998>}
 weapon_r {parent=hand_r, loc=<0.1975 3.4111 -0.3811>, rot=<0 0 0 -1>}
 lowerarm_twist_01_r {parent=lowerarm_r, loc=<-18.3161 0 0>, rot=<0 0 0 1>}
 dyn_sleeve_strap_r_MSTR {parent=lowerarm_twist_01_r, loc=<-0.2358 2.522 -3.5792>, rot=<0.6315 -0.7494 -0.121 -0.158>}
 dyn_sleeve_strap_r_01 {parent=dyn_sleeve_strap_r_MSTR, loc=<0.3422 0.0005 -0.0004>, rot=<0 0 0 -1>}
 dyn_sleeve_strap_r_02 {parent=dyn_sleeve_strap_r_01, loc=<3.716 -0.0005 0.0004>, rot=<0 0 0 -1>}
 lowerarm_twist_02_r {parent=lowerarm_r, loc=<-10.6707 0 0>, rot=<0 0 0 1>}
 upperarm_twist_01_r {parent=upperarm_r, loc=<-7.2135 0.1134 0>, rot=<0 0 0 1>}
 upperarm_twist_02_r {parent=upperarm_r, loc=<-19.7998 0.1134 0>, rot=<0 0 0 1>}
 neck_01 {parent=spine_05, loc=<11.3138 0.1529 0>, rot=<0 0 -0.2003 -0.9797>}
 neck_02 {parent=neck_01, loc=<4.5627 0 0>, rot=<0 0 0.0504 -0.9987>}
 head {parent=neck_02, loc=<5.3914 0 0>, rot=<0 0 0.115 -0.9934>}
 dyn_collar_back_MSTR {parent=spine_05, loc=<10.8806 -4.708 0>, rot=<-0.7053 0.0503 0.0503 -0.7053>}
 dyn_collar_back {parent=dyn_collar_back_MSTR, loc=<0.5899 0 0>, rot=<-0.0091 0.1273 -0.024 -0.9915>}
 dyn_collar_bk_l_MSTR {parent=spine_05, loc=<10.7868 -2.8711 -6.0998>, rot=<0.9115 0.0347 -0.1359 0.3868>}
 dyn_collar_bk_l {parent=dyn_collar_bk_l_MSTR, loc=<0.5899 0 0>, rot=<-0.0033 0.1518 0.0313 -0.9879>}
 dyn_collar_l_MSTR {parent=spine_05, loc=<8.0272 2.3027 -8.6986>, rot=<0.9394 0.2122 -0.2046 0.1753>}
 dyn_collar_l {parent=dyn_collar_l_MSTR, loc=<0.8292 0 0>, rot=<-0.0814 0.2545 0.157 -0.9508>}
 dyn_collar_bk_r_MSTR {parent=spine_05, loc=<10.7868 -2.8712 6.0998>, rot=<-0.3868 0.1359 -0.0347 -0.9115>}
 dyn_collar_bk_r {parent=dyn_collar_bk_r_MSTR, loc=<0.59 0 0>, rot=<-0.0397 0.1657 -0.0425 -0.9845>}
 dyn_collar_r_MSTR {parent=spine_05, loc=<8.0273 2.3027 8.6986>, rot=<-0.1753 0.2046 -0.2122 -0.9394>}
 dyn_collar_r {parent=dyn_collar_r_MSTR, loc=<0.8295 -0.0001 -0.0001>, rot=<0.0453 0.1695 -0.0827 -0.981>}
 dyn_chest_strap_top_l_MSTR {parent=spine_04, loc=<3.0831 6.7725 -11.8854>, rot=<0.3028 0.5866 0.5975 0.4552>}
 dyn_chest_strap_top_01_l {parent=dyn_chest_strap_top_l_MSTR, loc=<0.4483 0 0>, rot=<-0.0442 -0.0651 -0.1249 -0.989>}
 dyn_chest_strap_top_02_l {parent=dyn_chest_strap_top_01_l, loc=<3.6186 0 0>, rot=<0.0437 -0.0395 0.103 -0.9929>}
 dyn_jacket_fr_l_MSTR {parent=spine_04, loc=<1.5362 3.8831 -12.0361>, rot=<0.0413 -0.9927 -0.1136 0.0047>}
 dyn_jacket_fr_l_01 {parent=dyn_jacket_fr_l_MSTR, loc=<0.259 0 0>, rot=<-0.0415 0.0496 0.019 -0.9977>}
 dyn_jacket_fr_l_02 {parent=dyn_jacket_fr_l_01, loc=<5.1309 0 0>, rot=<0.0429 0.0475 -0.0217 -0.9977>}
 dyn_chest_strap_bott_01_l {parent=dyn_jacket_fr_l_02, loc=<5.1379 0.2343 -0.1827>, rot=<0.556 0.302 -0.3281 -0.7014>}
 dyn_chest_strap_bott_02_l {parent=dyn_chest_strap_bott_01_l, loc=<3.783 0.0002 0.0004>, rot=<0.0438 -0.0437 -0.0257 -0.9978>}
 dyn_jacket_bk_MSTR {parent=spine_04, loc=<2.0436 -8.2788 -0.0002>, rot=<-0.006 -0.7071 0.7071 0.006>}
 dyn_jacket_bk_01 {parent=dyn_jacket_bk_MSTR, loc=<0.259 0 0>, rot=<-0.043 0.1045 -0.0075 -0.9936>}
 dyn_jacket_bk_02 {parent=dyn_jacket_bk_01, loc=<4.4853 0 0>, rot=<0.0447 0.01 -0.002 -0.9989>}
 dyn_jacket_bk_l_MSTR {parent=spine_04, loc=<1.0941 -7.2154 -6.2973>, rot=<0.0163 -0.8343 0.551 0.011>}
 dyn_jacket_bk_l_01 {parent=dyn_jacket_bk_l_MSTR, loc=<0.259 0 0>, rot=<0.0715 0.1892 -0.0539 -0.9779>}
 dyn_jacket_bk_l_02 {parent=dyn_jacket_bk_l_01, loc=<3.7592 0 0>, rot=<-0.0392 -0.0813 -0.025 -0.9956>}
 dyn_jacket_l_MSTR {parent=spine_04, loc=<-0.0982 -2.3958 -10.8868>, rot=<0.0373 -0.9911 0.1247 0.0268>}
 dyn_jacket_l_01 {parent=dyn_jacket_l_MSTR, loc=<0.259 0 0>, rot=<-0.058 0.2244 0.0438 -0.9718>}
 dyn_jacket_l_02 {parent=dyn_jacket_l_01, loc=<3.9344 0 0>, rot=<0.0434 -0.184 -0.0305 -0.9815>}
 dyn_chest_strap_top_r_MSTR {parent=spine_04, loc=<3.0833 6.7725 11.8854>, rot=<-0.4552 -0.5975 -0.5866 -0.3028>}
 dyn_chest_strap_top_01_r {parent=dyn_chest_strap_top_r_MSTR, loc=<0.4483 0 0.0002>, rot=<-0.0502 -0.0985 -0.0159 -0.9937>}
 dyn_chest_strap_top_02_r {parent=dyn_chest_strap_top_01_r, loc=<3.6186 0 -0.0002>, rot=<-0.0506 0.0104 -0.1731 -0.9836>}
 dyn_jacket_fr_r_MSTR {parent=spine_04, loc=<1.5361 3.8831 12.0361>, rot=<-0.0047 0.1136 0.9927 -0.0413>}
 dyn_jacket_fr_r_01 {parent=dyn_jacket_fr_r_MSTR, loc=<0.259 0 0>, rot=<0.0746 0.0953 -0.0743 -0.9899>}
 dyn_jacket_fr_r_02 {parent=dyn_jacket_fr_r_01, loc=<5.131 0 0>, rot=<-0.059 -0.0455 0.1437 -0.9868>}
 dyn_chest_strap_bott_01_r {parent=dyn_jacket_fr_r_02, loc=<4.8616 -0.4762 -0.7575>, rot=<-0.5627 0.3249 0.302 -0.6975>}
 dyn_chest_strap_bott_02_r {parent=dyn_chest_strap_bott_01_r, loc=<3.171 0.0156 0.0071>, rot=<-0.0453 -0.0173 -0.1696 -0.9843>}
 dyn_jacket_bk_r_MSTR {parent=spine_04, loc=<1.094 -7.2153 6.2973>, rot=<-0.011 -0.551 0.8343 -0.0163>}
 dyn_jacket_bk_r_01 {parent=dyn_jacket_bk_r_MSTR, loc=<0.2588 0 0>, rot=<0.0626 0.2785 -0.0299 -0.9579>}
 dyn_jacket_bk_r_02 {parent=dyn_jacket_bk_r_01, loc=<3.7592 0 0>, rot=<-0.0524 -0.2487 0.0616 -0.9652>}
 dyn_jacket_r_MSTR {parent=spine_04, loc=<-0.0977 -2.3958 10.8868>, rot=<-0.0268 -0.1247 0.9911 -0.0373>}
 dyn_jacket_r_01 {parent=dyn_jacket_r_MSTR, loc=<0.2595 0 0>, rot=<-0.0563 0.2183 0.008 -0.9742>}
 dyn_jacket_r_02 {parent=dyn_jacket_r_01, loc=<3.9341 0 0>, rot=<0.0426 -0.135 -0.0451 -0.9889>}
 thigh_l {parent=pelvis, loc=<-1.0855 -0.2109 -8.7053>, rot=<0.0009 0.0355 0.0155 -0.9993>}
 calf_l {parent=thigh_l, loc=<-42.5283 -0.0019 -0.7297>, rot=<0.0006 -0.0161 -0.0391 -0.9991>}
 calf_twist_01_l {parent=calf_l, loc=<-35.8483 -1.9197 -0.5794>, rot=<0.0237 0.0095 -0.0035 -0.9997>}
 calf_twist_02_l {parent=calf_l, loc=<-19.1546 -1.7943 -0.2638>, rot=<0.0237 0.0095 -0.0035 -0.9997>}
 foot_l {parent=calf_l, loc=<-46.7401 -2.5014 -0.592>, rot=<0.0427 -0.0238 0.0353 -0.9982>}
 ball_l {parent=foot_l, loc=<-3.5658 10.8066 -0.1544>, rot=<0.0001 -0.0001 -0.7063 -0.7079>}
 dyn_kneepad_l {parent=calf_l, loc=<-4.0829 2.6277 0.451>, rot=<-0.0137 0.0025 -0.1376 -0.9904>}
 dyn_shoelace_l_MSTR {parent=dyn_kneepad_l, loc=<3.7172 1.5986 0.2664>, rot=<-0.1219 -0.7032 -0.6925 -0.1055>}
 dyn_shoelace_out_l_01 {parent=dyn_shoelace_l_MSTR, loc=<0.564 -0.6422 -0.0401>, rot=<-0.0694 0.0054 0.1666 -0.9836>}
 dyn_shoelace_out_l_02 {parent=dyn_shoelace_out_l_01, loc=<3.8296 0 0>, rot=<0 0 -0.0436 -0.999>}
 dyn_shoelace_in_l_01 {parent=dyn_shoelace_l_MSTR, loc=<0.4899 0.403 -0.0213>, rot=<0.0585 0.0136 -0.0964 -0.9935>}
 dyn_shoelace_in_l_02 {parent=dyn_shoelace_in_l_01, loc=<3.8295 0 0>, rot=<0 0 0.0436 -0.999>}
 thigh_twist_01_l {parent=thigh_l, loc=<-27.952 0 0>, rot=<0.0013 0 0 -1>}
 thigh_r {parent=pelvis, loc=<-1.0853 -0.2109 8.7053>, rot=<0.0355 -0.0009 0.9993 0.0155>}
 calf_r {parent=thigh_r, loc=<42.5286 0.0019 0.7297>, rot=<0.0006 -0.0161 -0.0391 -0.9991>}
 calf_twist_01_r {parent=calf_r, loc=<35.8485 1.9197 0.5794>, rot=<0.0237 0.0095 -0.0035 -0.9997>}
 calf_twist_02_r {parent=calf_r, loc=<19.1546 1.7943 0.2638>, rot=<0.0237 0.0095 -0.0035 -0.9997>}
 foot_r {parent=calf_r, loc=<46.7403 2.5014 0.5919>, rot=<0.0427 -0.0238 0.0353 -0.9982>}
 ball_r {parent=foot_r, loc=<3.5658 -10.8066 0.1544>, rot=<0.0001 -0.0001 -0.7063 -0.7079>}
 dyn_kneepad_r {parent=calf_r, loc=<4.0828 -2.6277 -0.4512>, rot=<0.129 -0.9915 -0.0131 0.0019>}
 dyn_shoelace_r_MSTR {parent=dyn_kneepad_r, loc=<3.7088 -1.6183 0.2762>, rot=<0.1192 -0.7034 0.6929 -0.1046>}
 dyn_shoelace_in_r_01 {parent=dyn_shoelace_r_MSTR, loc=<0.4899 -0.403 -0.0213>, rot=<-0.0594 0.0071 0.096 -0.9936>}
 dyn_shoelace_in_r_02 {parent=dyn_shoelace_in_r_01, loc=<3.8295 -0.0001 0>, rot=<0 0 -0.0436 -0.999>}
 dyn_shoelace_out_r_01 {parent=dyn_shoelace_r_MSTR, loc=<0.564 0.6422 -0.0401>, rot=<0.0707 -0.0005 -0.1661 -0.9836>}
 dyn_shoelace_out_r_02 {parent=dyn_shoelace_out_r_01, loc=<3.8295 0 0>, rot=<0 0 0.0436 -0.999>}
 thigh_twist_01_r {parent=thigh_r, loc=<27.952 0 0>, rot=<0.0013 0 0 -1>}
 dyn_belt {parent=pelvis, loc=<2.1056 -0.3262 -0.0747>, rot=<0.9987 0.0272 -0.0094 -0.0421>}
 dyn_belt_shaper_r {parent=dyn_belt, loc=<-1.5459 -8.4618 -7.3255>, rot=<0 0 0 1>}
 dyn_belt_shaper_l {parent=dyn_belt, loc=<0.7666 -6.8567 8.7008>, rot=<0 0 0 1>}
 dyn_pouch_r_MSTR {parent=dyn_belt, loc=<-1.5258 1.2602 -14.462>, rot=<-0.0006 -0.9923 -0.0323 -0.1199>}
 dyn_pouch_r {parent=dyn_pouch_r_MSTR, loc=<0 0 0.6269>, rot=<0 0 0 1>}
 dyn_strap_hip_01_r {parent=dyn_pouch_r, loc=<6.2947 0.7991 -0.1658>, rot=<0.1182 0.0208 0.0286 -0.9924>}
 dyn_strap_hip_02_r {parent=dyn_strap_hip_01_r, loc=<8.2925 0 0>, rot=<0 -0.0199 -0.0004 -0.9998>}
 dyn_strap_hip_02_r1 {parent=dyn_strap_hip_02_r, loc=<9.0084 0 0>, rot=<0 -0.0121 -0.0003 -0.9999>}
 dyn_pouch_l_MSTR {parent=dyn_belt, loc=<1.8672 -0.3214 12.8171>, rot=<0 -0.1339 0 -0.991>}
 dyn_pouch_l {parent=dyn_pouch_l_MSTR, loc=<0 0 0.3735>, rot=<0 0 0 1>}
 dyn_skirt_front_MSTR {parent=pelvis, loc=<-8.1688 8.0048 0>, rot=<0 0.7071 0.7071 0>}
 dyn_skirt_front {parent=dyn_skirt_front_MSTR, loc=<0.5524 0 0.0198>, rot=<0 0 0 -1>}
 dyn_skirt_back_MSTR {parent=pelvis, loc=<-7.7645 -10.6358 0>, rot=<0 -0.7071 0.7071 0>}
 dyn_skirt_back {parent=dyn_skirt_back_MSTR, loc=<0.5524 0 -0.0198>, rot=<0 0 0 1>}
 attach {parent=root, loc=<0 0 0>, rot=<0 0 0 1>}
 ik_foot_root {parent=root, loc=<0 0 0>, rot=<0 0 0 1>}
 ik_foot_l {parent=ik_foot_root, loc=<14.8262 -7.3545 8.296>, rot=<0.0357 0.7045 0.0284 -0.7082>}
 ik_foot_r {parent=ik_foot_root, loc=<-14.8262 -7.3546 8.2958>, rot=<0.7082 0.0284 -0.7045 0.0357>}
 ik_hand_root {parent=root, loc=<0 0 0>, rot=<0 0 0 1>}
 ik_hand_gun {parent=ik_hand_root, loc=<-48.1744 3.3546 104.4381>, rot=<0.6276 0.1309 0.4183 -0.6434>}
 ik_hand_l {parent=ik_hand_gun, loc=<59.3272 67.6997 34.3524>, rot=<-0.6158 -0.7027 -0.3565 0>}
 ik_hand_r {parent=ik_hand_gun, loc=<0 0 0>, rot=<0 0 0 1>}
 dyn_root_simspace {parent=root, loc=<0 0 0>, rot=<0 0 0 1>}
]
