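# Physics asset for secondary motion of the jacket and chest straps on the female medium Modern Witch player body.
# Two-segment capsule chains hang from tiny MSTR root bodies on spine_04, with larger pelvis-to-upper-spine capsules forming the torso.
# Bottom straps instead continue from the front jacket chains.
# Chain joints have limited swing and tight twist, straps narrower than jacket panels, while root and spine joints rotate freely.
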PhysicsAsset /Game/Characters/Player/Female/Medium/Bodies/F_MED_Modern_Witch/Meshes/F_MED_Modern_Witch_Dyn
bodies {
 dyn_chest_strap_bott_01_l [{
   shapeType=Capsule
   length=1.4596
   radius=1.324
   shapeName=dyn_chest_strap_bott_01_l_capsule
   center=<2.6113 0.0087 -0.1181>
   rotation=<rot 7.8429 -89.9999 -88.5962>
  }]
 dyn_chest_strap_bott_01_r [{
   shapeType=Capsule
   length=1.4558
   radius=1.4538
   shapeName=dyn_chest_strap_bott_01_r_capsule
   center=<2.6801 -0.0952 -0.0302>
   rotation=<rot -7.7385 -90 -88.696>
  }]
 dyn_chest_strap_bott_02_l [{
   shapeType=Capsule
   length=3.0163
   radius=1.4601
   shapeName=dyn_chest_strap_bott_02_l_capsule
   center=<2.7633 0.1017 -0.6094>
   rotation=<rot 0 -93.4911 -78.0149>
  }]
 dyn_chest_strap_bott_02_r [{
   shapeType=Capsule
   length=1.3392
   radius=1.475
   shapeName=dyn_chest_strap_bott_02_r_capsule
   center=<1.9335 -0.2261 -0.1087>
   rotation=<rot -7.1856 -89.9999 -80.884>
  }]
 dyn_chest_strap_top_01_l [{
   shapeType=Capsule
   length=1.0814
   radius=1.25
   shapeName=dyn_chest_strap_top_01_l_capsule
   center=<2.0162 -0.2143 0.1477>
   rotation=<rot 13.731 -90 -87.0024>
  }]
 dyn_chest_strap_top_01_r [{
   shapeType=Capsule
   length=0.8693
   radius=1.25
   shapeName=dyn_chest_strap_top_01_r_capsule
   center=<2.4435 -0.2463 0.2021>
   rotation=<rot -9.8682 -89.9999 -89.1982>
  }]
 dyn_chest_strap_top_02_l [{
   shapeType=Capsule
   length=2.4506
   radius=1.25
   shapeName=dyn_chest_strap_top_02_l_capsule
   center=<2.1473 0.1927 -0.1292>
   rotation=<rot 0 -73.3676 -72.536>
  }]
 dyn_chest_strap_top_02_r [{
   shapeType=Capsule
   length=1.0497
   radius=1.1138
   shapeName=dyn_chest_strap_top_02_r_capsule
   center=<1.3008 -0.3651 0.1874>
   rotation=<rot 2.4662 -90 -92.689>
  }]
 dyn_chest_strap_top_l_MSTR [{shapeType=Capsule, length=0.2, radius=0.2}]
 dyn_chest_strap_top_r_MSTR [{shapeType=Capsule, length=0.2, radius=0.2}]
 dyn_jacket_bk_01 [{
   shapeType=Capsule
   length=1.505
   radius=2.005
   shapeName=dyn_jacket_bk_01_capsule
   center=<2.6166 -0.015 0.1213>
   rotation=<rot 3.597 -95.0473 -96.7564>
  }]
 dyn_jacket_bk_02 [{
   shapeType=Capsule
   length=2.8431
   radius=2.1622
   shapeName=dyn_jacket_bk_02_capsule
   center=<2.4435 -0.0214 -0.4067>
   rotation=<rot 0.0335 90.7075 -89.9995>
  }]
 dyn_jacket_bk_MSTR [{shapeType=Capsule, length=0.2, radius=0.2}]
 dyn_jacket_bk_l_01 [{
   shapeType=Capsule
   length=2.8128
   radius=2.2198
   shapeName=dyn_jacket_bk_l_01_capsule
   center=<2.1582 -0.4405 0.1074>
   rotation=<rot -22.9056 -86.8285 -85.7266>
  }]
 dyn_jacket_bk_l_02 [{
   shapeType=Capsule
   length=2.487
   radius=2.2313
   shapeName=dyn_jacket_bk_l_02_capsule
   center=<2.3726 -0.3688 -0.3787>
   rotation=<rot -19.9388 -82.0634 -89.6333>
  }]
 dyn_jacket_bk_l_MSTR [{shapeType=Capsule, length=0.2, radius=0.2}]
 dyn_jacket_bk_r_01 [{
   shapeType=Capsule
   length=2.4622
   radius=2.2009
   shapeName=dyn_jacket_bk_r_01_capsule
   center=<1.5596 -0.1821 -0.0333>
   rotation=<rot 5.2603 -86.4215 -89.0697>
  }]
 dyn_jacket_bk_r_02 [{
   shapeType=Capsule
   length=2.7987
   radius=2.0365
   shapeName=dyn_jacket_bk_r_02_capsule
   center=<2.2272 0.0883 -0.0485>
   rotation=<rot 0.5369 91.4748 -90.0946>
  }]
 dyn_jacket_bk_r_MSTR [{shapeType=Capsule, length=0.2, radius=0.2}]
 dyn_jacket_fr_l_01 [{
   shapeType=Capsule
   length=3.1818
   radius=2.539
   shapeName=dyn_jacket_fr_l_01_capsule
   center=<2.0145 -0.276 0.7127>
   rotation=<rot -4.7962 -85.3308 -88.4507>
  }]
 dyn_jacket_fr_l_02 [{
   shapeType=Capsule
   length=2.7203
   radius=2.4642
   shapeName=dyn_jacket_fr_l_02_capsule
   center=<3.7606 0.0015 -0.243>
   rotation=<rot 0 91.903 -85.5417>
  }]
 dyn_jacket_fr_l_MSTR [{shapeType=Capsule, length=0.2, radius=0.2}]
 dyn_jacket_fr_r_01 [{
   shapeType=Capsule
   length=3.3771
   radius=2.1754
   shapeName=dyn_jacket_fr_r_01_capsule
   center=<2.0127 -0.099 0.3471>
   rotation=<rot 4.673 -84.5874 -87.0192>
  }]
 dyn_jacket_fr_r_02 [{
   shapeType=Capsule
   length=2.7704
   radius=2.5359
   shapeName=dyn_jacket_fr_r_02_capsule
   center=<3.5353 0.4627 -0.4058>
   rotation=<rot 9.9609 -93.8695 -84.885>
  }]
 dyn_jacket_fr_r_MSTR [{shapeType=Capsule, length=0.2, radius=0.2}]
 dyn_jacket_l_01 [{
   shapeType=Capsule
   length=2.8656
   radius=2
   shapeName=dyn_jacket_l_01_capsule
   center=<2.1979 0.1633 0.014>
   rotation=<rot 8.0467 86.6098 -87.0504>
  }]
 dyn_jacket_l_02 [{
   shapeType=Capsule
   length=3.3707
   radius=2.1049
   shapeName=dyn_jacket_l_02_capsule
   center=<3.037 0.1032 -0.1742>
   rotation=<rot 7.0794 96.3474 -85.888>
  }]
 dyn_jacket_l_MSTR [{shapeType=Capsule, length=0.2, radius=0.2}]
 dyn_jacket_r_01 [{
   shapeType=Capsule
   length=2.9019
   radius=1.9661
   shapeName=dyn_jacket_r_01_capsule
   center=<2.1427 0.2738 -0.0033>
   rotation=<rot -11.2199 91.4502 -85.8595>
  }]
 dyn_jacket_r_02 [{
   shapeType=Capsule
   length=3.4816
   radius=2.1722
   shapeName=dyn_jacket_r_02_capsule
   center=<1.9514 -0.14 -0.3171>
   rotation=<rot -5.1235 96.2131 -88.1298>
  }]
 dyn_jacket_r_MSTR [{shapeType=Capsule, length=0.2, radius=0.2}]
 pelvis [{shapeType=Capsule, length=0.2, radius=0.2}]
 spine_02 [{shapeType=Capsule, length=6.5938, radius=7.2188, center=<2.2873 1.348 0.0002>}]
 spine_03 [{shapeType=Capsule, length=6.5938, radius=7.2188, center=<1.619 1.3191 -0.0001>}]
 spine_04 [
  {shapeType=Capsule, length=8.0938, radius=7.2188, center=<0.666 6.4456 0.0001>}
  {shapeType=Capsule, length=5.3596, radius=7.6965, center=<0 0.9334 0>}
 ]
}
constraints [
 {bone1Name=dyn_chest_strap_top_01_l, bone2Name=dyn_chest_strap_top_l_MSTR, swing1LimitDegrees=20, swing2LimitDegrees=10, twistLimitDegrees=5}
 {bone1Name=dyn_chest_strap_top_02_l, bone2Name=dyn_chest_strap_top_01_l, swing1LimitDegrees=20, swing2LimitDegrees=5, twistLimitDegrees=5}
 {bone1Name=dyn_jacket_fr_l_01, bone2Name=dyn_jacket_fr_l_MSTR, swing1LimitDegrees=35, swing2LimitDegrees=35, twistLimitDegrees=5}
 {bone1Name=dyn_jacket_fr_l_02, bone2Name=dyn_jacket_fr_l_01, swing1LimitDegrees=35, swing2LimitDegrees=35, twistLimitDegrees=5}
 {bone1Name=dyn_chest_strap_bott_01_l, bone2Name=dyn_jacket_fr_l_02, swing1LimitDegrees=20, swing2LimitDegrees=10, twistLimitDegrees=5}
 {bone1Name=dyn_chest_strap_bott_02_l, bone2Name=dyn_chest_strap_bott_01_l, swing1LimitDegrees=20, swing2LimitDegrees=5, twistLimitDegrees=5}
 {bone1Name=dyn_jacket_bk_01, bone2Name=dyn_jacket_bk_MSTR, swing1LimitDegrees=35, swing2LimitDegrees=35, twistLimitDegrees=5}
 {bone1Name=dyn_jacket_bk_02, bone2Name=dyn_jacket_bk_01, swing1LimitDegrees=35, swing2LimitDegrees=35, twistLimitDegrees=5}
 {bone1Name=dyn_jacket_bk_l_01, bone2Name=dyn_jacket_bk_l_MSTR, swing1LimitDegrees=35, swing2LimitDegrees=35, twistLimitDegrees=5}
 {bone1Name=dyn_jacket_bk_l_02, bone2Name=dyn_jacket_bk_l_01, swing1LimitDegrees=35, swing2LimitDegrees=35, twistLimitDegrees=5}
 {bone1Name=dyn_jacket_l_01, bone2Name=dyn_jacket_l_MSTR, swing1LimitDegrees=35, swing2LimitDegrees=35, twistLimitDegrees=5}
 {bone1Name=dyn_jacket_l_02, bone2Name=dyn_jacket_l_01, swing1LimitDegrees=35, swing2LimitDegrees=35, twistLimitDegrees=5}
 {bone1Name=dyn_chest_strap_top_01_r, bone2Name=dyn_chest_strap_top_r_MSTR, swing1LimitDegrees=20, swing2LimitDegrees=10, twistLimitDegrees=5}
 {bone1Name=dyn_chest_strap_top_02_r, bone2Name=dyn_chest_strap_top_01_r, swing1LimitDegrees=20, swing2LimitDegrees=5, twistLimitDegrees=5}
 {bone1Name=dyn_jacket_fr_r_01, bone2Name=dyn_jacket_fr_r_MSTR, swing1LimitDegrees=35, swing2LimitDegrees=35, twistLimitDegrees=5}
 {bone1Name=dyn_jacket_fr_r_02, bone2Name=dyn_jacket_fr_r_01, swing1LimitDegrees=35, swing2LimitDegrees=35, twistLimitDegrees=5}
 {bone1Name=dyn_chest_strap_bott_01_r, bone2Name=dyn_jacket_fr_r_02, swing1LimitDegrees=20, swing2LimitDegrees=10, twistLimitDegrees=5}
 {bone1Name=dyn_chest_strap_bott_02_r, bone2Name=dyn_chest_strap_bott_01_r, swing1LimitDegrees=20, swing2LimitDegrees=5, twistLimitDegrees=5}
 {bone1Name=dyn_jacket_bk_r_01, bone2Name=dyn_jacket_bk_r_MSTR, swing1LimitDegrees=35, swing2LimitDegrees=35, twistLimitDegrees=5}
 {bone1Name=dyn_jacket_bk_r_02, bone2Name=dyn_jacket_bk_r_01, swing1LimitDegrees=35, swing2LimitDegrees=35, twistLimitDegrees=5}
 {bone1Name=dyn_jacket_r_01, bone2Name=dyn_jacket_r_MSTR, swing1LimitDegrees=35, swing2LimitDegrees=35, twistLimitDegrees=5}
 {bone1Name=dyn_jacket_r_02, bone2Name=dyn_jacket_r_01, swing1LimitDegrees=35, swing2LimitDegrees=35, twistLimitDegrees=5}
 {
  bone1Name=dyn_chest_strap_top_l_MSTR
  bone2Name=spine_04
  swing1LimitDegrees=45
  swing1Motion=Free
  swing2LimitDegrees=45
  swing2Motion=Free
  twistLimitDegrees=45
  twistMotion=Free
 }
 {
  bone1Name=dyn_jacket_fr_l_MSTR
  bone2Name=spine_04
  swing1LimitDegrees=45
  swing1Motion=Free
  swing2LimitDegrees=45
  swing2Motion=Free
  twistLimitDegrees=45
  twistMotion=Free
 }
 {
  bone1Name=dyn_jacket_bk_MSTR
  bone2Name=spine_04
  swing1LimitDegrees=45
  swing1Motion=Free
  swing2LimitDegrees=45
  swing2Motion=Free
  twistLimitDegrees=45
  twistMotion=Free
 }
 {
  bone1Name=dyn_jacket_bk_l_MSTR
  bone2Name=spine_04
  swing1LimitDegrees=45
  swing1Motion=Free
  swing2LimitDegrees=45
  swing2Motion=Free
  twistLimitDegrees=45
  twistMotion=Free
 }
 {
  bone1Name=dyn_jacket_l_MSTR
  bone2Name=spine_04
  swing1LimitDegrees=45
  swing1Motion=Free
  swing2LimitDegrees=45
  swing2Motion=Free
  twistLimitDegrees=45
  twistMotion=Free
 }
 {
  bone1Name=dyn_chest_strap_top_r_MSTR
  bone2Name=spine_04
  swing1LimitDegrees=45
  swing1Motion=Free
  swing2LimitDegrees=45
  swing2Motion=Free
  twistLimitDegrees=45
  twistMotion=Free
 }
 {
  bone1Name=dyn_jacket_fr_r_MSTR
  bone2Name=spine_04
  swing1LimitDegrees=45
  swing1Motion=Free
  swing2LimitDegrees=45
  swing2Motion=Free
  twistLimitDegrees=45
  twistMotion=Free
 }
 {
  bone1Name=dyn_jacket_bk_r_MSTR
  bone2Name=spine_04
  swing1LimitDegrees=45
  swing1Motion=Free
  swing2LimitDegrees=45
  swing2Motion=Free
  twistLimitDegrees=45
  twistMotion=Free
 }
 {
  bone1Name=dyn_jacket_r_MSTR
  bone2Name=spine_04
  swing1LimitDegrees=45
  swing1Motion=Free
  swing2LimitDegrees=45
  swing2Motion=Free
  twistLimitDegrees=45
  twistMotion=Free
 }
 {
  bone1Name=spine_04
  bone2Name=spine_03
  swing1LimitDegrees=45
  swing1Motion=Free
  swing2LimitDegrees=45
  swing2Motion=Free
  twistLimitDegrees=45
  twistMotion=Free
 }
 {
  bone1Name=spine_03
  bone2Name=spine_02
  swing1LimitDegrees=45
  swing1Motion=Free
  swing2LimitDegrees=45
  swing2Motion=Free
  twistLimitDegrees=45
  twistMotion=Free
 }
 {
  bone1Name=spine_02
  bone2Name=pelvis
  swing1LimitDegrees=45
  swing1Motion=Free
  swing2LimitDegrees=45
  swing2Motion=Free
  twistLimitDegrees=45
  twistMotion=Free
 }
]
