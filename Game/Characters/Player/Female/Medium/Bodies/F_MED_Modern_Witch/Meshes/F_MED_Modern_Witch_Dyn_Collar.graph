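# Physics asset for the dynamic collar on the Modern Witch medium female player body, simulating five collar pieces: back, back-left, back-right, left and right.
# Each piece has a fitted capsule constrained to a tiny capsule on its MSTR bone, with wider swing limits on the sides than at the back.
# Head and neck_02 capsules, linked by an angularly free constraint, provide body shapes around the collar.
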PhysicsAsset /Game/Characters/Player/Female/Medium/Bodies/F_MED_Modern_Witch/Meshes/F_MED_Modern_Witch_Dyn_Collar
bodies {
 dyn_collar_back [{
   shapeType=Capsule
   length=7.1402
   radius=2.6329
   shapeName=dyn_collar_back_capsule
   center=<7.4745 0.5791 2.3476>
   rotation=<rot 5.6183 -87.9736 -94.006>
  }]
 dyn_collar_back_MSTR [{shapeType=Capsule, length=0.2, radius=0.2}]
 dyn_collar_bk_l [{
   shapeType=Capsule
   length=5.6479
   radius=3.1768
   shapeName=dyn_collar_bk_l_capsule
   center=<6.4845 -0.1124 -0.7789>
   rotation=<rot 3.538 -94.8698 -80.2162>
  }]
 dyn_collar_bk_l_MSTR [{shapeType=Capsule, length=0.2, radius=0.2}]
 dyn_collar_bk_r [{
   shapeType=Capsule
   length=7.8639
   radius=2.5039
   shapeName=dyn_collar_bk_r_capsule
   center=<4.3624 0.2176 0.5381>
   rotation=<rot 0 -89.9616 -89.5521>
  }]
 dyn_collar_bk_r_MSTR [{shapeType=Capsule, length=0.2, radius=0.2}]
 dyn_collar_l [{
   shapeType=Capsule
   length=8.7005
   radius=3.4726
   shapeName=dyn_collar_l_capsule
   center=<8.762 0.2652 2.97>
   rotation=<rot -0.0167 -92.2973 -106.2207>
  }]
 dyn_collar_l_MSTR [{shapeType=Capsule, length=0.2, radius=0.2}]
 dyn_collar_r [{
   shapeType=Capsule
   length=7.9344
   radius=2.5629
   shapeName=dyn_collar_r_capsule
   center=<5.8668 -0.6002 1.0344>
   rotation=<rot 79.9904 179.9999 180>
  }]
 dyn_collar_r_MSTR [{shapeType=Capsule, length=0.2, radius=0.2}]
 head [{shapeType=Capsule, length=8.9688, radius=8.25, center=<3.8778 0 0>, rotation=<rot 90 0 0>}]
 neck_02 [{shapeType=Capsule, length=2, radius=2, rotation=<rot 90 0 0>}]
}
constraints [
 {bone1Name=dyn_collar_back, bone2Name=dyn_collar_back_MSTR, swing1LimitDegrees=25, swing2LimitDegrees=25, twistLimitDegrees=20}
 {bone1Name=dyn_collar_bk_l, bone2Name=dyn_collar_bk_l_MSTR, swing1LimitDegrees=25, swing2LimitDegrees=25, twistLimitDegrees=20}
 {bone1Name=dyn_collar_l, bone2Name=dyn_collar_l_MSTR, swing1LimitDegrees=60, swing2LimitDegrees=60, twistLimitDegrees=20}
 {bone1Name=dyn_collar_bk_r, bone2Name=dyn_collar_bk_r_MSTR, swing1LimitDegrees=25, swing2LimitDegrees=25, twistLimitDegrees=20}
 {bone1Name=dyn_collar_r, bone2Name=dyn_collar_r_MSTR, swing1LimitDegrees=60, swing2LimitDegrees=60, twistLimitDegrees=20}
 {
  bone1Name=head
  bone2Name=neck_02
  swing1LimitDegrees=45
  swing1Motion=Free
  swing2LimitDegrees=45
  swing2Motion=Free
  twistLimitDegrees=45
  twistMotion=Free
 }
]
