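# Physics asset for the left leg of the medium female Modern Witch player body, covering a dynamic kneepad bone.
# It holds two capsule bodies: a small one on the left calf and a larger, offset and rotated one on the kneepad.
# A single constraint joins kneepad to calf with swing 2 and twist locked, leaving swing 1 as the only rotational axis, so the pad acts like a hinge.
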PhysicsAsset /Game/Characters/Player/Female/Medium/Bodies/F_MED_Modern_Witch/Meshes/F_MED_Modern_Witch_Dyn_Leg_l
bodies {
 calf_l [{shapeType=Capsule, length=2, radius=2}]
 dyn_kneepad_l [{
   shapeType=Capsule
   length=4.0562
   radius=4.6996
   shapeName=dyn_kneepad_l_capsule
   center=<5.7423 0.6339 -0.0544>
   rotation=<rot -3.5423 -69.9622 79.7146>
  }]
}
constraints [{
  bone1Name=dyn_kneepad_l
  bone2Name=calf_l
  swing1LimitDegrees=45
  swing2LimitDegrees=5
  swing2Motion=Locked
  twistLimitDegrees=5
  twistMotion=Locked
 }]
