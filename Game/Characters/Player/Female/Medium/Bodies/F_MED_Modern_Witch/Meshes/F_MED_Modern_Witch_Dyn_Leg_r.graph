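# Physics asset for the right-leg dynamic kneepad on the medium female Modern Witch player body.
# It has two capsule bodies: a small one on calf_r and a larger, offset and rotated one on dyn_kneepad_r.
# A single constraint joins the kneepad to the calf with twist and swing 2 locked, so the pad can rotate only about the swing 1 axis, like a hinge.
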PhysicsAsset /Game/Characters/Player/Female/Medium/Bodies/F_MED_Modern_Witch/Meshes/F_MED_Modern_Witch_Dyn_Leg_r
bodies {
 calf_r [{shapeType=Capsule, length=2, radius=2}]
 dyn_kneepad_r [{
   shapeType=Capsule
   length=4.3062
   radius=3.6996
   shapeName=dyn_kneepad_r_capsule
   center=<5.3801 0.0073 0.1629>
   rotation=<rot 0.9772 -104.8104 80.3723>
  }]
}
constraints [{
  bone1Name=dyn_kneepad_r
  bone2Name=calf_r
  swing1LimitDegrees=45
  swing2LimitDegrees=45
  swing2Motion=Locked
  twistLimitDegrees=45
  twistMotion=Locked
 }]
