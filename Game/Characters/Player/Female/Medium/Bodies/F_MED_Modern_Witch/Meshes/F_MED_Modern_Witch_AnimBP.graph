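# Post-process style AnimBlueprint for the Modern Witch female body part, adding secondary motion to collar, belt, pouches, skirt, straps, kneepads and shoelaces.
# The pose copied from the parent mesh gets corrective bone offsets and constraints, driven by event-graph alphas for arms, head, crouch, jump and sprint.
# RigidBody and AnimDynamics nodes then simulate cloth and accessories, before a Control Rig finishes the pose.
# A computed rotation orients a dedicated sim-space root bone.
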AnimBlueprint /Game/Characters/Player/Female/Medium/Bodies/F_MED_Modern_Witch/Meshes/F_MED_Modern_Witch_AnimBP
target_skeleton=&/F_MED_Modern_Witch_Skeleton
graphs [
 AnimGraph {nodes [
   "dyn_pouch_l_MSTR.rotateZ = thigh_l.rotateZ * 0.2 - Bone Driven Controller"!(ComponentPose:Component Space Pose Link Structure, Alpha:float=1) -> (Pose:Component Space Pose Link Structure) {
    is=title AnimGraphNode_BoneDrivenController
    bAffectTargetRotationZ=true
    multiplier=0.2
    sourceComponent=RotationZ
    alphaBoolBlend {blendOption=Linear}
    sourceBone {boneName=thigh_l}
    targetBone {boneName=dyn_pouch_l_MSTR}
   }
   Component To Local()
   n1 = Anim Dynamics - Bone: dyn_strap_hip_01_r(ComponentPose=Local To Component(LocalPose=Use cached pose 'Presim'()), Alpha=1, ExternalForce="0,0,0") {
    angularDampingOverride=0.9999
    bOverrideAngularDamping=true
    bUseGravityOverride=true
    simulationSpace=RootRelative
    alphaBoolBlend {blendOption=Linear}
    boundBone {boneName=dyn_strap_hip_01_r}
    componentAppliedLinearAccClamp=<90000 90000 90000>
    componentLinearAccScale=<0.5 0.5 0.5>
    componentLinearVelScale=<0.5 0.5 0.5>
    gravityOverride {z=-980}
    physicsBodyDefinitions [{
      boundBone {boneName=dyn_strap_hip_01_r}
      boxExtents=<5 20 20>
      constraintSetup {angularLimitsMax=<10 -5 0>, angularLimitsMin=<-10 -90 0>, angularTarget=<0 0 0>}
      localJointOffset=<5 0 0>
     }]
   }
   n2 = Anim Dynamics - Bone: dyn_strap_hip_02_r(ComponentPose=@n1, Alpha=1, ExternalForce="0,0,0") {
    angularDampingOverride=0.9999
    bOverrideAngularDamping=true
    simulationSpace=RootRelative
    alphaBoolBlend {blendOption=Linear}
    boundBone {boneName=dyn_strap_hip_02_r}
    componentAppliedLinearAccClamp=<90000 90000 90000>
    componentLinearAccScale=<0.5 0.5 0.5>
    componentLinearVelScale=<0.5 0.5 0.5>
    physicsBodyDefinitions [{
      boundBone {boneName=dyn_strap_hip_02_r}
      boxExtents=<5 20 20>
      constraintSetup {angularLimitsMax=<10 0 20>, angularLimitsMin=<-10 -90 0>, angularTarget=<0 0 0>}
      localJointOffset=<5 0 0>
     }]
   }
   n3 = Anim Dynamics - Bone: dyn_strap_hip_02_r1(ComponentPose=@n2, Alpha=1, ExternalForce="0,0,0") {
    angularDampingOverride=0.9999
    bOverrideAngularDamping=true
    simulationSpace=RootRelative
    alphaBoolBlend {blendOption=Linear}
    boundBone {boneName=dyn_strap_hip_02_r1}
    componentAppliedLinearAccClamp=<90000 90000 90000>
    componentLinearAccScale=<0.5 0.5 0.5>
    componentLinearVelScale=<0.5 0.5 0.5>
    physicsBodyDefinitions [{
      boundBone {boneName=dyn_strap_hip_02_r1}
      boxExtents=<5 20 20>
      constraintSetup {angularLimitsMax=<10 0 30>, angularLimitsMin=<-10 -90 -30>, angularTarget=<0 0 0>}
      localJointOffset=<5 0 0>
     }]
   }
   n4 = Anim Dynamics - Bone: dyn_kneepad_l(ComponentPose=@n3, Alpha=0.3, ExternalForce="0,0,0") {
    angularDampingOverride=0.9999
    angularSpringConstant=500
    bAngularSpring=true
    bGravityOverrideInSimSpace=true
    bOverrideAngularDamping=true
    bUseGravityOverride=true
    alphaBoolBlend {blendOption=Linear}
    boundBone {boneName=dyn_kneepad_l}
    componentAppliedLinearAccClamp=<2000 2000 7000>
    componentLinearAccScale=<0.5 0.5 0.5>
    componentLinearVelScale=<0.5 0.5 0.5>
    gravityOverride {z=-980}
    physicsBodyDefinitions [{
      sphereCollisionRadius=0
      boundBone {boneName=dyn_kneepad_l}
      boxExtents=<12 22 22>
      constraintSetup {angularLimitsMax=<5 5 10>, angularLimitsMin=<-5 -5 -10>, angularTarget=<0 0 0>}
      localJointOffset=<6 0 0>
     }]
   }
   n5 = Anim Dynamics - Bone: dyn_kneepad_r(ComponentPose=@n4, Alpha=0.2, ExternalForce="0,0,0") {
    angularDampingOverride=0.9999
    angularSpringConstant=500
    bAngularSpring=true
    bGravityOverrideInSimSpace=true
    bOverrideAngularDamping=true
    bUseGravityOverride=true
    alphaBoolBlend {blendOption=Linear}
    boundBone {boneName=dyn_kneepad_r}
    componentAppliedLinearAccClamp=<2000 2000 7000>
    componentLinearAccScale=<0.5 0.5 0.5>
    componentLinearVelScale=<0.5 0.5 0.5>
    gravityOverride {z=-980}
    physicsBodyDefinitions [{
      sphereCollisionRadius=0
      boundBone {boneName=dyn_kneepad_r}
      boxExtents=<12 22 22>
      constraintSetup {angularLimitsMax=<5 5 10>, angularLimitsMin=<-5 -5 -10>, angularTarget=<0 0 0>}
      localJointOffset=<6 0 0>
     }]
   }
   n6 = "Transform (Modify) Bone - Bone: dyn_kneepad_r"!(ComponentPose=@n5, bAlphaBoolEnabled:bool=True, Translation="0,0,0", Rotation="0.000,37,0.000", Scale="1,1,1") -> (Pose) {
    alphaInputType=Bool
    rotationMode=BMM_Additive
    rotationSpace=BCS_BoneSpace
    scaleMode=BMM_Additive
    scaleSpace=BCS_BoneSpace
    translationMode=BMM_Additive
    translationSpace=BCS_BoneSpace
    alphaBoolBlend {blendInTime=0.1, blendOption=Linear, blendOutTime=0.1}
    alphaScaleBiasClamp {bInterpResult=true, interpSpeedDecreasing=15, interpSpeedIncreasing=15}
    boneToModify {boneName=dyn_kneepad_r}
    rotation {yaw=37}
   }
   n7 = "Constraint - dyn_shoelace_l_MSTR (PPP)"!(ComponentPose=@n6, Alpha=1, ConstraintWeights_0=0.701, ConstraintWeights_1:float=0.2471, ConstraintWeights_2:float=0.0518) -> (Pose) {
    alphaBoolBlend {blendOption=Linear}
    boneToModify {boneName=dyn_shoelace_l_MSTR}
    constraintSetup [{targetBone {boneName=dyn_kneepad_l}}, {targetBone {boneName=calf_l}}, {targetBone {boneName=thigh_l}}]
    constraintWeights [1,1,1]
   }
   n8 = "Constraint - dyn_shoelace_r_MSTR (PPP)"!(ComponentPose=@n7, Alpha=1, ConstraintWeights_0=0.0518, ConstraintWeights_1:float=0.701, ConstraintWeights_2:float=0.2471) -> (Pose) {
    alphaBoolBlend {blendOption=Linear}
    boneToModify {boneName=dyn_shoelace_r_MSTR}
    constraintSetup [{targetBone {boneName=thigh_r}}, {targetBone {boneName=dyn_kneepad_r}}, {targetBone {boneName=calf_r}}]
    constraintWeights [1,1,1]
   }
   n9 = Anim Dynamics - Bone: dyn_shoelace_out_l_01!(ComponentPose=@n8, Alpha=1, GravityOverride:Vector="0,0,0", ExternalForce="0,0,0") -> (Pose) {
    angularDampingOverride=1
    angularSpringConstant=1500
    bAngularSpring=true
    bGravityOverrideInSimSpace=true
    bOverrideAngularDamping=true
    gravityScale=0.35
    simulationSpace=RootRelative
    alphaBoolBlend {blendOption=Linear}
    alphaScaleBias {scale=2}
    boundBone {boneName=dyn_shoelace_out_l_01}
    physicsBodyDefinitions [{
      sphereCollisionRadius=0
      boundBone {boneName=dyn_shoelace_out_l_01}
      boxExtents=<10 20 20>
      constraintSetup {angularLimitsMax=<5 10 5>, angularLimitsMin=<-5 -35 -5>, angularTarget=<0 0 0>}
      localJointOffset=<10 0 0>
     }]
   }
   n10 = Anim Dynamics - Bone: dyn_shoelace_out_l_02!(ComponentPose=@n9, Alpha=1, GravityOverride:Vector="0,0,0", ExternalForce="0,0,0") -> (Pose) {
    angularDampingOverride=1
    angularSpringConstant=1500
    bAngularSpring=true
    bGravityOverrideInSimSpace=true
    bOverrideAngularDamping=true
    gravityScale=0.35
    simulationSpace=RootRelative
    alphaBoolBlend {blendOption=Linear}
    alphaScaleBias {scale=2}
    boundBone {boneName=dyn_shoelace_out_l_02}
    physicsBodyDefinitions [{
      sphereCollisionRadius=0
      boundBone {boneName=dyn_shoelace_out_l_02}
      boxExtents=<10 20 20>
      constraintSetup {angularLimitsMax=<5 10 5>, angularLimitsMin=<-5 -35 -5>, angularTarget=<0 0 0>}
      localJointOffset=<10 0 0>
     }]
   }
   n11 = Anim Dynamics - Bone: dyn_shoelace_in_l_01!(ComponentPose=@n10, Alpha=1, GravityOverride:Vector="0,0,0", ExternalForce="0,0,0") -> (Pose) {
    angularDampingOverride=1
    angularSpringConstant=1500
    bAngularSpring=true
    bGravityOverrideInSimSpace=true
    bOverrideAngularDamping=true
    gravityScale=0.35
    simulationSpace=RootRelative
    alphaBoolBlend {blendOption=Linear}
    alphaScaleBias {scale=2}
    boundBone {boneName=dyn_shoelace_in_l_01}
    physicsBodyDefinitions [{
      sphereCollisionRadius=0
      boundBone {boneName=dyn_shoelace_in_l_01}
      boxExtents=<10 20 20>
      constraintSetup {angularLimitsMax=<5 10 5>, angularLimitsMin=<-5 -35 -5>, angularTarget=<0 0 0>}
      localJointOffset=<10 0 0>
     }]
   }
   n12 = Anim Dynamics - Bone: dyn_shoelace_in_l_02!(ComponentPose=@n11, Alpha=1, GravityOverride:Vector="0,0,0", ExternalForce="0,0,0") -> (Pose) {
    angularDampingOverride=1
    angularSpringConstant=1500
    bAngularSpring=true
    bGravityOverrideInSimSpace=true
    bOverrideAngularDamping=true
    gravityScale=0.35
    simulationSpace=RootRelative
    alphaBoolBlend {blendOption=Linear}
    alphaScaleBias {scale=2}
    boundBone {boneName=dyn_shoelace_in_l_02}
    physicsBodyDefinitions [{
      sphereCollisionRadius=0
      boundBone {boneName=dyn_shoelace_in_l_02}
      boxExtents=<10 20 20>
      constraintSetup {angularLimitsMax=<5 10 5>, angularLimitsMin=<-5 -35 -5>, angularTarget=<0 0 0>}
      localJointOffset=<10 0 0>
     }]
   }
   n13 = Anim Dynamics - Bone: dyn_shoelace_in_r_01!(ComponentPose=@n12, Alpha=1, GravityOverride:Vector="0,0,0", ExternalForce="0,0,0") -> (Pose) {
    angularDampingOverride=1
    angularSpringConstant=1500
    bAngularSpring=true
    bGravityOverrideInSimSpace=true
    bOverrideAngularDamping=true
    gravityScale=0.35
    simulationSpace=RootRelative
    alphaBoolBlend {blendOption=Linear}
    alphaScaleBias {scale=2}
    boundBone {boneName=dyn_shoelace_in_r_01}
    physicsBodyDefinitions [{
      sphereCollisionRadius=0
      boundBone {boneName=dyn_shoelace_in_r_01}
      boxExtents=<10 20 20>
      constraintSetup {angularLimitsMax=<5 10 5>, angularLimitsMin=<-5 -35 -5>, angularTarget=<0 0 0>}
      localJointOffset=<10 0 0>
     }]
   }
   n14 = Anim Dynamics - Bone: dyn_shoelace_in_r_02!(ComponentPose=@n13, Alpha=1, GravityOverride:Vector="0,0,0", ExternalForce="0,0,0") -> (Pose) {
    angularDampingOverride=1
    angularSpringConstant=1500
    bAngularSpring=true
    bGravityOverrideInSimSpace=true
    bOverrideAngularDamping=true
    gravityScale=0.35
    simulationSpace=RootRelative
    alphaBoolBlend {blendOption=Linear}
    alphaScaleBias {scale=2}
    boundBone {boneName=dyn_shoelace_in_r_02}
    physicsBodyDefinitions [{
      sphereCollisionRadius=0
      boundBone {boneName=dyn_shoelace_in_r_02}
      boxExtents=<10 20 20>
      constraintSetup {angularLimitsMax=<5 10 5>, angularLimitsMin=<-5 -35 -5>, angularTarget=<0 0 0>}
      localJointOffset=<10 0 0>
     }]
   }
   n15 = Anim Dynamics - Bone: dyn_shoelace_out_r_01!(ComponentPose=@n14, Alpha=1, GravityOverride:Vector="0,0,0", ExternalForce="0,0,0") -> (Pose) {
    angularDampingOverride=1
    angularSpringConstant=1500
    bAngularSpring=true
    bGravityOverrideInSimSpace=true
    bOverrideAngularDamping=true
    gravityScale=0.35
    simulationSpace=RootRelative
    alphaBoolBlend {blendOption=Linear}
    alphaScaleBias {scale=2}
    boundBone {boneName=dyn_shoelace_out_r_01}
    physicsBodyDefinitions [{
      sphereCollisionRadius=0
      boundBone {boneName=dyn_shoelace_out_r_01}
      boxExtents=<10 20 20>
      constraintSetup {angularLimitsMax=<5 10 5>, angularLimitsMin=<-5 -35 -5>, angularTarget=<0 0 0>}
      localJointOffset=<10 0 0>
     }]
   }
   n16 = Anim Dynamics - Bone: dyn_shoelace_out_r_02!(ComponentPose=@n15, Alpha=1, GravityOverride:Vector="0,0,0", ExternalForce="0,0,0") -> (Pose) {
    angularDampingOverride=1
    angularSpringConstant=1500
    bAngularSpring=true
    bGravityOverrideInSimSpace=true
    bOverrideAngularDamping=true
    gravityScale=0.35
    simulationSpace=RootRelative
    alphaBoolBlend {blendOption=Linear}
    alphaScaleBias {scale=2}
    boundBone {boneName=dyn_shoelace_out_r_02}
    physicsBodyDefinitions [{
      sphereCollisionRadius=0
      boundBone {boneName=dyn_shoelace_out_r_02}
      boxExtents=<10 20 20>
      constraintSetup {angularLimitsMax=<5 10 5>, angularLimitsMin=<-5 -35 -5>, angularTarget=<0 0 0>}
      localJointOffset=<10 0 0>
     }]
   }
   n17 = Control Rig(Source=Component To Local(ComponentPose=@n16)) {
    alphaBoolBlend {blendOption=Linear}
    controlRigAssetReference {blueprintRigClass=&/F_MED_Modern_Witch_CtrlRig.F_MED_Modern_Witch_CtrlRig_C}
   }
   Output Pose(Result=@n17) {layerGroup=None}
   Save cached pose 'StartCache'(Pose=Copy Pose From Mesh() {bUseAttachedParent=true}) {cacheName=StartCache}
   n18 = Get arm_up_l() -> (arm_up_l:double)
   n19 = "Transform (Modify) Bone - Bone: dyn_collar_l_MSTR"(
    ComponentPose=Local To Component(LocalPose=Use cached pose 'StartCache'())
    Alpha=@n18
    Translation="-0.146,0.339,-0.368"
    Rotation="-22,-0.648,5"
    Scale="1,1,1"
   ) {
    rotationMode=BMM_Additive
    rotationSpace=BCS_BoneSpace
    translationMode=BMM_Additive
    translationSpace=BCS_BoneSpace
    alphaBoolBlend {blendOption=Linear}
    alphaScaleBiasClamp {bInterpResult=true, interpSpeedDecreasing=15, interpSpeedIncreasing=15}
    boneToModify {boneName=dyn_collar_l_MSTR}
    rotation=<rot -22 -0.648 5>
    translation=<-0.146 0.339 -0.368>
   }
   n20 = "Transform (Modify) Bone - Bone: dyn_collar_bk_l_MSTR"(ComponentPose=@n19, Alpha=@n18, Translation="-0.211,0.000,-1.263", Rotation="-6.972,0.000,0.000", Scale="1,1,1") {
    rotationMode=BMM_Additive
    rotationSpace=BCS_BoneSpace
    translationMode=BMM_Additive
    translationSpace=BCS_BoneSpace
    alphaBoolBlend {blendOption=Linear}
    alphaScaleBiasClamp {bInterpResult=true, interpSpeedDecreasing=15, interpSpeedIncreasing=15}
    boneToModify {boneName=dyn_collar_bk_l_MSTR}
    rotation {pitch=-6.972}
    translation=<-0.211 0 -1.2628>
   }
   n21 = "Transform (Modify) Bone - Bone: dyn_collar_r_MSTR"(ComponentPose=@n20, Alpha=Get arm_up_r() -> (arm_up_r:double), Translation="0,0,0", Rotation="10,0,0", Scale="1,1,1") {
    rotationMode=BMM_Additive
    rotationSpace=BCS_BoneSpace
    translationMode=BMM_Additive
    translationSpace=BCS_BoneSpace
    alphaBoolBlend {blendOption=Linear}
    alphaScaleBiasClamp {bInterpResult=true, interpSpeedDecreasing=15, interpSpeedIncreasing=15}
    boneToModify {boneName=dyn_collar_r_MSTR}
    rotation {pitch=10}
   }
   Save cached pose 'withRigidCache'(Pose=Blend Poses by bool(BlendPose_0=Component To Local(ComponentPose=@n21), BlendPose_1=Use cached pose 'StartCache'(), bActiveValue=Get bRigidBodyEnabled()) {blendType=Linear}) {cacheName=withRigidCache}
   n22 = Get head_up/down() -> (head_up/down:double)
   n23 = Get crouch alpha() -> (crouch alpha:double)
   n24 = "Transform (Modify) Bone - Bone: dyn_root_simspace"(
    ComponentPose=Local To Component(LocalPose=Use cached pose 'withRigidCache'())
    Alpha=1
    Translation="0,0,0"
    Rotation=Get sim_space_rotation() -> (sim_space_rotation:Rotator)
    Scale="1,1,1"
   ) {
    rotationMode=BMM_Replace
    rotationSpace=BCS_WorldSpace
    translationSpace=BCS_BoneSpace
    alphaBoolBlend {blendOption=Linear}
    alphaScaleBiasClamp {bInterpResult=true, interpSpeedDecreasing=2, interpSpeedIncreasing=2}
    boneToModify {boneName=dyn_root_simspace}
   }
   n25 = "Transform (Modify) Bone - Bone: dyn_belt"(
    ComponentPose=@n24
    Alpha=Get jumping alpha() -> (jumping alpha:double)
    Translation="6,-0.232,1.331"
    Rotation="11.275,-8,0.539"
    Scale="1,1,1"
   ) {
    rotationMode=BMM_Additive
    rotationSpace=BCS_BoneSpace
    translationMode=BMM_Additive
    translationSpace=BCS_BoneSpace
    alphaBoolBlend {blendOption=Linear}
    alphaScaleBiasClamp {bInterpResult=true}
    boneToModify {boneName=dyn_belt}
    rotation=<rot 11.275 -8 0.539>
    translation=<6 -0.232 1.331>
   }
   n26 = "Transform (Modify) Bone - Bone: dyn_belt_shaper_l"(ComponentPose=@n25, Alpha=@n23, Translation="3.508,-2.107,-0.269", Rotation="0,0,0", Scale="1,1,1") {
    rotationMode=BMM_Additive
    rotationSpace=BCS_BoneSpace
    translationMode=BMM_Additive
    translationSpace=BCS_BoneSpace
    alphaBoolBlend {blendOption=Linear}
    alphaScaleBiasClamp {bInterpResult=true, interpSpeedDecreasing=15, interpSpeedIncreasing=15}
    boneToModify {boneName=dyn_belt_shaper_l}
    translation=<3.508 -2.107 -0.269>
   }
   n27 = "Transform (Modify) Bone - Bone: dyn_belt_shaper_r"(ComponentPose=@n26, Alpha=@n23, Translation="3.794,-0.356,-0.000", Rotation="0,0,0", Scale="1,1,1") {
    rotationMode=BMM_Additive
    rotationSpace=BCS_BoneSpace
    translationMode=BMM_Additive
    translationSpace=BCS_BoneSpace
    alphaBoolBlend {blendOption=Linear}
    alphaScaleBiasClamp {bInterpResult=true, interpSpeedDecreasing=15, interpSpeedIncreasing=15}
    boneToModify {boneName=dyn_belt_shaper_r}
    translation=<3.794 -0.3557 0>
   }
   n28 = "Transform (Modify) Bone - Bone: dyn_belt"(ComponentPose=@n27, Alpha=@n23, Translation="4.625,0.091,0.000", Rotation="0.000,-5.208,0.000", Scale="1,1.05,1") {
    rotationMode=BMM_Additive
    rotationSpace=BCS_BoneSpace
    scaleMode=BMM_Additive
    scaleSpace=BCS_BoneSpace
    translationMode=BMM_Additive
    translationSpace=BCS_BoneSpace
    alphaBoolBlend {blendOption=Linear}
    alphaScaleBiasClamp {bInterpResult=true, interpSpeedDecreasing=15, interpSpeedIncreasing=15}
    boneToModify {boneName=dyn_belt}
    rotation {yaw=-5.208}
    scale {y=1.05}
    translation=<4.625 0.091>
   }
   n29 = "Constraint - dyn_pouch_r_MSTR (PPPP)"!(
    ComponentPose=@n28
    Alpha=1
    ConstraintWeights_0=0.0175
    ConstraintWeights_1:float=0.0564
    ConstraintWeights_2:float=0.2183
    ConstraintWeights_3:float=0.7078
   ) -> (Pose) {
    alphaBoolBlend {blendOption=Linear}
    boneToModify {boneName=dyn_pouch_r_MSTR}
    constraintSetup [{targetBone {boneName=thigh_r}}, {targetBone {boneName=pelvis}}, {targetBone {boneName=thigh_twist_01_r}}, {targetBone {boneName=dyn_belt}}]
    constraintWeights [1,1,1,1]
   }
   n30 = "Constraint - dyn_pouch_l_MSTR (PPP)"!(ComponentPose=@n29, Alpha=1, ConstraintWeights_0=0.0192, ConstraintWeights_1:float=0.1179, ConstraintWeights_2:float=0.8629) -> (Pose) {
    alphaBoolBlend {blendOption=Linear}
    boneToModify {boneName=dyn_pouch_l_MSTR}
    constraintSetup [{targetBone {boneName=dyn_belt_shaper_l}}, {targetBone {boneName=thigh_twist_01_l}}, {targetBone {boneName=pelvis}}]
    constraintWeights [1,1,1]
   }
   n31 = "Transform (Modify) Bone - Bone: dyn_pouch_l_MSTR"(ComponentPose=@n30, Alpha=@n23, Translation="0.000,3.027,0.000", Rotation="0.000,0.000,13.985", Scale="1,1,1") {
    rotationMode=BMM_Additive
    rotationSpace=BCS_BoneSpace
    translationMode=BMM_Additive
    translationSpace=BCS_BoneSpace
    alphaBoolBlend {blendOption=Linear}
    alphaScaleBiasClamp {bInterpResult=true, interpSpeedDecreasing=15, interpSpeedIncreasing=15}
    boneToModify {boneName=dyn_pouch_l_MSTR}
    rotation {roll=13.9851}
    translation=<0 3.0269 0>
   }
   n32 = Get Sprinting Alpha() -> (Sprinting Alpha:double)
   n33 = "Transform (Modify) Bone - Bone: dyn_pouch_r_MSTR"(ComponentPose=@n31, Alpha=@n32, Translation="0,0,0", Rotation="6.337,12,12.238", Scale="1,1,1") {
    rotationMode=BMM_Additive
    rotationSpace=BCS_BoneSpace
    translationMode=BMM_Additive
    translationSpace=BCS_BoneSpace
    alphaBoolBlend {blendOption=Linear}
    alphaScaleBiasClamp {bInterpResult=true, interpSpeedDecreasing=15, interpSpeedIncreasing=15}
    boneToModify {boneName=dyn_pouch_r_MSTR}
    rotation=<rot 6.337 12 12.238>
   }
   n34 = "Transform (Modify) Bone - Bone: dyn_belt"(ComponentPose=@n33, Alpha=@n32, Translation="2.112,-0.000,-0.000", Rotation="0,0,0", Scale="1.000,1.080,1.000") {
    rotationMode=BMM_Additive
    rotationSpace=BCS_BoneSpace
    scaleMode=BMM_Additive
    scaleSpace=BCS_BoneSpace
    translationMode=BMM_Additive
    translationSpace=BCS_BoneSpace
    alphaBoolBlend {blendOption=Linear}
    alphaScaleBiasClamp {bInterpResult=true, interpSpeedDecreasing=15, interpSpeedIncreasing=15}
    boneToModify {boneName=dyn_belt}
    scale {y=1.0802}
    translation {x=2.112}
   }
   n35 = RigidBody(ComponentPose=@n34) {
    bClampLinearTranslationLimitToRefPose=true
    bOverrideWorldGravity=true
    overridePhysicsAsset=&/F_MED_Modern_Witch_Dyn
    simulationSpace=BaseBoneSpace
    alphaBoolBlend {blendOption=Linear}
    baseBoneRef {boneName=spine_03}
    componentAppliedLinearAccClamp=<3000 3000 3000>
    componentLinearAccScale=<0.4 0.3 0.3>
    componentLinearVelScale=<0.4 0.3 0.3>
    overrideWorldGravity {z=-980}
   }
   n36 = RigidBody(ComponentPose=@n35) {
    bClampLinearTranslationLimitToRefPose=true
    bOverrideWorldGravity=true
    overridePhysicsAsset=&/F_MED_Modern_Witch_Dyn_Collar
    simulationSpace=BaseBoneSpace
    alphaBoolBlend {blendOption=Linear}
    baseBoneRef {boneName=spine_05}
    componentAppliedLinearAccClamp=<3000 3000 3000>
    componentLinearAccScale=<0.4 0.3 0.3>
    componentLinearVelScale=<0.4 0.3 0.3>
    overrideWorldGravity {z=-980}
   }
   n37 = "Transform (Modify) Bone - Bone: dyn_collar_back"(ComponentPose=@n36, Alpha=@n22, Translation="0,0,0", Rotation="0,0,-50", Scale="1,1,1") {
    rotationMode=BMM_Additive
    scaleMode=BMM_Additive
    translationMode=BMM_Additive
    alphaBoolBlend {blendOption=Linear}
    boneToModify {boneName=dyn_collar_back}
    rotation {roll=-50}
   }
   n38 = "Transform (Modify) Bone - Bone: dyn_collar_bk_l"(ComponentPose=@n37, Alpha=@n22, Translation="0,0,0", Rotation="0,0,-35", Scale="1,1,1") {
    rotationMode=BMM_Additive
    scaleMode=BMM_Additive
    translationMode=BMM_Additive
    alphaBoolBlend {blendOption=Linear}
    boneToModify {boneName=dyn_collar_bk_l}
    rotation {roll=-35}
   }
   n39 = "Transform (Modify) Bone - Bone: dyn_collar_bk_r"(ComponentPose=@n38, Alpha=@n22, Translation="0,0,0", Rotation="0,0,-50", Scale="1,1,1") {
    rotationMode=BMM_Additive
    scaleMode=BMM_Additive
    translationMode=BMM_Additive
    alphaBoolBlend {blendOption=Linear}
    boneToModify {boneName=dyn_collar_bk_r}
    rotation {roll=-50}
   }
   n40 = "Transform (Modify) Bone - Bone: dyn_collar_r"(ComponentPose=@n39, Alpha=@n22, Translation="0,0,0", Rotation="0,0,-10", Scale="1,1,1") {
    rotationMode=BMM_Additive
    scaleMode=BMM_Additive
    translationMode=BMM_Additive
    alphaBoolBlend {blendOption=Linear}
    boneToModify {boneName=dyn_collar_r}
    rotation {roll=-10}
   }
   n41 = RigidBody(ComponentPose=@n40) {
    bClampLinearTranslationLimitToRefPose=true
    bOverrideWorldGravity=true
    overridePhysicsAsset=&/F_MED_Modern_Witch_Dyn_Leg_l
    simulationSpace=BaseBoneSpace
    alphaBoolBlend {blendOption=Linear}
    baseBoneRef {boneName=calf_l}
    componentAppliedLinearAccClamp=<3000 3000 3000>
    componentLinearAccScale=<0.4 0.3 0.3>
    componentLinearVelScale=<0.4 0.3 0.3>
    overrideWorldGravity {z=-980}
   }
   n42 = RigidBody(ComponentPose=@n41) {
    bClampLinearTranslationLimitToRefPose=true
    bOverrideWorldGravity=true
    overridePhysicsAsset=&/F_MED_Modern_Witch_Dyn_Leg_r
    simulationSpace=BaseBoneSpace
    alphaBoolBlend {blendOption=Linear}
    baseBoneRef {boneName=calf_r}
    componentAppliedLinearAccClamp=<3000 3000 3000>
    componentLinearAccScale=<0.4 0.3 0.3>
    componentLinearVelScale=<0.4 0.3 0.3>
    overrideWorldGravity {z=-980}
   }
   n43 = RigidBody(ComponentPose=@n42) {
    bClampLinearTranslationLimitToRefPose=true
    bOverrideWorldGravity=true
    overridePhysicsAsset=&/F_MED_Modern_Witch_Dyn_skirt.F_MED_Modern_Witch_Dyn_Skirt
    simulationSpace=BaseBoneSpace
    alphaBoolBlend {blendOption=Linear}
    baseBoneRef {boneName=pelvis}
    componentLinearAccScale=<0.1 0.1 0.1>
    componentLinearVelScale=<0.1 0.1 0.1>
    overrideWorldGravity {z=-980}
   }
   n44 = "Transform (Modify) Bone - Bone: dyn_skirt_back"!(ComponentPose=@n43, bAlphaBoolEnabled:bool=Get bIsSkyDiving(), Translation="0,0,10", Rotation="0,0,0", Scale="1,1,1") -> (Pose) {
    alphaInputType=Bool
    rotationMode=BMM_Additive
    rotationSpace=BCS_BoneSpace
    scaleMode=BMM_Additive
    scaleSpace=BCS_BoneSpace
    translationMode=BMM_Additive
    translationSpace=BCS_BoneSpace
    alphaBoolBlend {blendOption=Linear}
    boneToModify {boneName=dyn_skirt_back}
    translation {z=10}
   }
   n45 = Anim Dynamics - Bone: dyn_sleeve_strap_r_MSTR(ComponentPose=@n44, Alpha=1, ExternalForce="0,0,0") {
    bGravityOverrideInSimSpace=true
    bUseGravityOverride=true
    simulationSpace=BoneRelative
    alphaBoolBlend {blendOption=Linear}
    boundBone {boneName=dyn_sleeve_strap_r_MSTR}
    chainEnd {boneName=dyn_sleeve_strap_r_01}
    componentLinearAccScale {z=0.01}
    gravityOverride {z=-980}
    physicsBodyDefinitions [
     {
      sphereCollisionRadius=0
      boundBone {boneName=dyn_sleeve_strap_r_MSTR}
      boxExtents=<10 50 50>
      constraintSetup {angularTargetAxis=AxisZ, twistAxis=AxisZ, angularLimitsMax=<5 10 5>, angularLimitsMin=<-5 -30 -5>, angularTarget=<0 0 0>}
      localJointOffset=<8 0 0>
     }
     {
      boundBone {boneName=dyn_sleeve_strap_r_01}
      boxExtents=<10 10 10>
      constraintSetup {angularLimitsMax=<0 0 0>, angularLimitsMin=<0 0 0>, angularTarget=<0 0 0>}
      localJointOffset=<0 0 0>
     }
    ]
    relativeSpaceBone {boneName=dyn_sleeve_strap_r_MSTR}
   }
   n46 = Anim Dynamics - Bone: dyn_sleeve_strap_l_MSTR(ComponentPose=@n45, Alpha=1, ExternalForce="0,0,0") {
    bGravityOverrideInSimSpace=true
    bUseGravityOverride=true
    simulationSpace=BoneRelative
    alphaBoolBlend {blendOption=Linear}
    boundBone {boneName=dyn_sleeve_strap_l_MSTR}
    chainEnd {boneName=dyn_sleeve_strap_r_01}
    componentLinearAccScale {z=0.01}
    gravityOverride {z=-980}
    physicsBodyDefinitions [{
      sphereCollisionRadius=0
      boundBone {boneName=dyn_sleeve_strap_l_MSTR}
      boxExtents=<10 50 50>
      constraintSetup {angularTargetAxis=AxisZ, twistAxis=AxisZ, angularLimitsMax=<5 10 5>, angularLimitsMin=<-5 -30 -5>, angularTarget=<0 0 0>}
      localJointOffset=<8 0 0>
     }]
    relativeSpaceBone {boneName=dyn_sleeve_strap_r_MSTR}
   }
   n47 = "Transform (Modify) Bone - Bone: dyn_pouch_l_MSTR"!(ComponentPose=@n46, bAlphaBoolEnabled:bool=True, Translation="0,0,1", Rotation="12,0,10", Scale="1,1,1") -> (Pose) {
    alphaInputType=Bool
    rotationMode=BMM_Additive
    rotationSpace=BCS_BoneSpace
    scaleMode=BMM_Additive
    scaleSpace=BCS_BoneSpace
    translationMode=BMM_Additive
    translationSpace=BCS_BoneSpace
    alphaBoolBlend {blendInTime=0.1, blendOption=Linear, blendOutTime=0.1}
    alphaScaleBiasClamp {bInterpResult=true, interpSpeedDecreasing=15, interpSpeedIncreasing=15}
    boneToModify {boneName=dyn_pouch_l_MSTR}
    rotation {pitch=12, roll=10}
    translation {z=1}
   }
   n48 = Anim Dynamics - Bone: dyn_shoelace_r_MSTR(ComponentPose=@n47, Alpha=1, ExternalForce="0,0,0") {
    bGravityOverrideInSimSpace=true
    bUseGravityOverride=true
    simulationSpace=World
    alphaBoolBlend {blendOption=Linear}
    boundBone {boneName=dyn_shoelace_r_MSTR}
    chainEnd {boneName=dyn_strap_hip_02_r}
    componentLinearVelScale=<0.5 0.5 0.5>
    gravityOverride {z=-980}
    physicsBodyDefinitions [{
      sphereCollisionRadius=0
      boundBone {boneName=dyn_shoelace_r_MSTR}
      boxExtents=<10 25 25>
      constraintSetup {angularLimitsMax=<0 15 30>, angularLimitsMin=<0 0 -10>, angularTarget=<0 0 0>}
      localJointOffset=<10 0 4>
     }]
    relativeSpaceBone {boneName=dyn_pouch_r}
   }
   Save cached pose 'Presim'(Pose=Component To Local(ComponentPose=@n48)) {cacheName=Presim}
  ]}
 EventGraph {nodes [
   Event BlueprintUpdateAnimation()
   > n1 = Sequence!(execute) -> (then_0, then_1, then_2:Exec, then_3:Exec, then_4:Exec, then_5:Exec)
   > n2 = Branch(Condition=Get bIsJumping())
   Set jumping alpha(execute=[@n2.then, @n3.then], jumping alpha:double=1) -> (Output_Get:double)
   n3 = Branch(execute=@n2.else, Condition=Get bIsFalling())
   Set jumping alpha(execute=@n3.else, jumping alpha=0)
   n4 = Branch(execute=@n1.then_1, Condition=Get bIsSprinting())
   > Set Sprinting Alpha(Sprinting Alpha:double=1) -> (Output_Get:double)
   Set Sprinting Alpha(execute=@n4.else, Sprinting Alpha=0)
   n5 = BreakTransform(InTransform=GetDeltaTransformFromRefPose(self=GetOwnerSkeletalMeshComponent(), BoneName=upperarm_l, BaseName=spine_05))
   Set arm_up_l(execute=@n1.then_2, arm_up_l:double=MapRangeClamped(Value=BreakRotator(InRot=@n5.Rotation).Yaw, InRangeA=60, InRangeB=100)) -> (Output_Get:double)
   n6 = BreakTransform(InTransform=GetDeltaTransformFromRefPose(self=GetOwnerSkeletalMeshComponent(), BoneName=upperarm_r, BaseName=spine_05))
   > Set arm_up_r(arm_up_r:double=MapRangeClamped(Value=BreakRotator(InRot=@n6.Rotation).Yaw, InRangeA=60, InRangeB=100)) -> (Output_Get:double)
   n7 = BreakTransform(InTransform=GetDeltaTransformFromRefPose(self=GetOwnerSkeletalMeshComponent(), BoneName=head, BaseName=neck_02))
   > Set head_up/down(head_up/down:double=MapRangeClamped(Value=BreakRotator(InRot=@n7.Rotation).Yaw, InRangeA=10, InRangeB=-23)) -> (Output_Get:double)
   n8 = Gravity Override State Function DEPRECATED(
    execute=@n1.then_3
    CharacterPartAnimInstance=Self-Reference()
    GameTimeInSeconds=GetGameTimeInSeconds()
    GravityOverrideParams=Get Gravity Override Params()
   )
   > Set GravityOverride(GravityOverride=@n8)
   n9 = ComplexSimSpaceFunction DEPRECATED!(
    execute=@n1.then_4
    CharacterPartAnimInstance:Custom Character Part Anim Instance Object Reference=Self-Reference()
    InterpSpeedMax:double=9
    PawnSpeedMax:double=20
    RInterpErrorTolerance:double=1
    DeltaTime:double=0
    DebugDraw:bool=false
    sim_space_rotation_in:Rotator=Get sim_space_rotation()
    inComponentSpaceIn:bool=Get InComponentSpace() -> (InComponentSpace:bool)
    inComponentSpaceLastFrameIn:bool=Get InComponentSpaceLastFrame() -> (InComponentSpaceLastFrame:bool)
    lockedOnSimSpaceIn:bool=Get LockedOnSimSpace() -> (LockedOnSimSpace:bool)
    InterpSpeedIn:double=Get InterpSpeed() -> (InterpSpeed:double)
   ) -> (
    then
    sim_space_rotation:Rotator
    inComponentSpaceOut:bool
    inComponentSpaceLastFrameOut:bool
    lockedOnSimSpaceOut:bool
    InterpSpeedOut:double
   )
   > Set sim_space_rotation(sim_space_rotation:Rotator=@n9) -> (Output_Get:Rotator)
   > Set InComponentSpace(InComponentSpace:bool=@n9.inComponentSpaceOut) -> (Output_Get:bool)
   > Set InComponentSpaceLastFrame(InComponentSpaceLastFrame:bool=@n9.inComponentSpaceLastFrameOut) -> (Output_Get:bool)
   > Set LockedOnSimSpace(LockedOnSimSpace:bool=@n9.lockedOnSimSpaceOut) -> (Output_Get:bool)
   > Set InterpSpeed(InterpSpeed:double=@n9.InterpSpeedOut) -> (Output_Get:double)
   n10 = Break Fort Anim Input Slope Info!(FortAnimInput_SlopeInfo:"Fort Anim Input Slope Info Structure (by ref)"=Get SlopeInfoInput() -> (SlopeInfoInput:Fort Anim Input Slope Info Structure)) -> (RootSlopeTranslation:Vector=~, LeftFootSlopeTranslation:Vector=~, RightFootSlopeTranslation:Vector=~, SlopeRotation:Rotator=~) {is=title K2Node_BreakStruct}
   Set slope_root_trans(execute=@n1.then_5, slope_root_trans:Vector=@n10) -> (Output_Get:Vector)
   > Set slope_foot_rot(slope_foot_rot:Rotator=@n10.SlopeRotation) -> (Output_Get:Rotator)
   Event OnBeginCrouching()
   > Set crouch alpha(crouch alpha:double=1) -> (Output_Get:double)
   Event OnEndCrouching()
   > Set crouch alpha(crouch alpha=0)
  ]}
]
variables [
 arm_up_l {type=number}
 arm_up_r {type=number}
 crouch alpha {type=number}
 Gravity Override Params {
  type=GravityOverrideParamsStruct
  default {
   skydive_DivingDown {gravityOverride=<0 0 0>}
   skydive_FloatingDown {gravityOverride=<0 0 250>, windAmplitude=<100 -250 200>, windFrequency=<1.5 0.5 2>}
   skydive_GoingUp {gravityOverride=<0 0 -100>, windAmplitude=<100 -250 200>}
   swimming {gravityOverride=<0 0 0>}
  }
 }
 GravityOverride {type=Vector}
 head_up/down {type=number}
 InComponentSpace {type=boolean}
 InComponentSpaceLastFrame {type=boolean}
 InterpSpeed {type=number}
 jumping alpha {type=number}
 LockedOnSimSpace {type=boolean}
 sim_space_rotation {type=Rotator}
 slope_foot_rot {type=Rotator}
 slope_root_trans {type=Vector}
 Sprinting Alpha {type=number}
]
