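# Physics asset for the dynamic skirt of the female medium Modern Witch player body.
# Front and back skirt bones each carry an identical capsule, constrained to the pelvis with narrow swing and twist limits.
# Both thighs have capsules on free-rotating pelvis constraints, giving leg volumes alongside the skirt, while the pelvis itself is only a tiny anchor capsule.
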PhysicsAsset /Game/Characters/Player/Female/Medium/Bodies/F_MED_Modern_Witch/Meshes/F_MED_Modern_Witch_Dyn_skirt.F_MED_Modern_Witch_Dyn_Skirt
bodies {
 dyn_skirt_back [{shapeType=Capsule, length=34.3262, radius=5.9623, shapeName=dyn_skirt_back_capsule, center=<7.8069 -1.1432 -1.2084>, rotation=<rot 0 0 -90>}]
 dyn_skirt_front [{shapeType=Capsule, length=34.3262, radius=5.9623, shapeName=dyn_skirt_back_capsule, center=<7.8069 -1.1432 -1.2084>, rotation=<rot 0 0 -90>}]
 pelvis [{shapeType=Capsule, length=0.2, radius=0.2}]
 thigh_l [{shapeType=Capsule, length=38.9936, radius=5.223, center=<-21.5087 -1.2162 -0.6483>, rotation=<rot 84.9736 -180 -180>}]
 thigh_r [{shapeType=Capsule, length=37.7436, radius=4.973, center=<21.5086 1.2162 0.6474>, rotation=<rot 84.9736 -180 -180>}]
}
constraints [
 {
  bone1Name=thigh_r
  bone2Name=pelvis
  swing1LimitDegrees=45
  swing1Motion=Free
  swing2LimitDegrees=45
  swing2Motion=Free
  twistLimitDegrees=45
  twistMotion=Free
 }
 {
  bone1Name=thigh_l
  bone2Name=pelvis
  swing1LimitDegrees=45
  swing1Motion=Free
  swing2LimitDegrees=45
  swing2Motion=Free
  twistLimitDegrees=45
  twistMotion=Free
 }
 {bone1Name=dyn_skirt_back, bone2Name=pelvis, swing1LimitDegrees=5, swing2LimitDegrees=15, twistLimitDegrees=5}
 {bone1Name=dyn_skirt_front, bone2Name=pelvis, swing1LimitDegrees=5, swing2LimitDegrees=15, twistLimitDegrees=5}
]
